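# Material function for the pirate ship set that layers puddles onto an incoming material where blue vertex color is painted.
# HeightLerp breaks up the painted mask with a swirl noise texture projected in world XY, with scale and contrast parameters.
# The mask flattens the base normal, then, clamped and scaled by Puddle Amount, drives a standard layer blend toward a zero-roughness layer with all other attributes left at defaults.
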MaterialFunction /Game/Environments/Sets/Pirate_Cove/Pirate_Ship/Materials/MaterialFunctions/MF_Pirate_Ship_Puddles
graph {nodes [
  n1 = GetMaterialAttributes(None=FunctionInput() {inputType=FunctionInput_MaterialAttributes}) {attributeGetTypes [#1]}
  n2 = VertexColor()
  n3 = Divide(
   A=ComponentMask(None=WorldPosition().XYZ) {g=true, r=true}
   B=ScalarParameter() {defaultValue=279.4971, group=Puddles, parameterName=Puddles Scale}
  )
  n4 = LinearInterpolate(
   B=MaterialFunctionCall(
    Transition Phase=@n2.B
    Height Texture=TextureSample(UVs=@n3) {samplerType=SAMPLERTYPE_LinearGrayscale, texture=/Game/Global/Textures/Default/Noise/T_SwirlNoise_H2}.RGB
    Contrast=ScalarParameter() {defaultValue=0.1, group=Puddles, parameterName=PuddlesContrast}
   ) {materialFunction=&HeightLerp}.Alpha
   Alpha=@n2.B
  )
  n5 = LinearInterpolate(A=@n1.Normal, B=Constant3Vector() {constant=<rgba 0 0 1 0>}, Alpha=@n4)
  n6 = Constant()
  n7 = ScalarParameter() {defaultValue=1, group=Puddles, parameterName=Puddle Amount}
  n8 = Clamp(None=@n4)
  FunctionOutput(None=MaterialFunctionCall!(
    Base Material=SetMaterialAttributes!(MaterialAttributes=@n1.MaterialAttributes, Normal=@n5) {attributeSetTypes [#1]}
    Top Material=MakeMaterialAttributes!(
     BaseColor
     Metallic
     Specular
     Roughness=@n6
     Anisotropy
     EmissiveColor
     Opacity
     OpacityMask
     Normal
     Tangent
     WorldPositionOffset
     SubsurfaceColor
     ClearCoat
     ClearCoatRoughness
     AmbientOcclusion
     Refraction
     CustomizedUVs_0
     CustomizedUVs_1
     CustomizedUVs_2
     CustomizedUVs_3
     CustomizedUVs_4
     CustomizedUVs_5
     CustomizedUVs_6
     CustomizedUVs_7
     PixelDepthOffset
     ShadingModel
     Displacement
    )
    Alpha=Multiply(A=@n7, B=@n8)
   ) {materialFunction=/Engine/Functions/MaterialLayerFunctions/MatLayerBlend_Standard}.Blended Material)
 ]}
