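# Material function for the Pirate Cove ship that multiplies incoming base color by a world-space vertical two-color gradient.
# The gradient uses MF_WS_GridUV basement height, offset by MPC_PirateShip's GradientPosition and scaled by floor count.
# A transition factor, defaulting to UV1 red, selects per-texture-layer toggles controlling where the tint applies.
# Static switches bypass the effect or break it up with a world-aligned hull mask.
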MaterialFunction /Game/Environments/Sets/Pirate_Cove/Pirate_Ship/Materials/MaterialFunctions/MF_Pirate_Ship_WSGradient
graph {nodes [
  n1 = CollectionParameter() {collection=&/MPC_PirateShip, parameterName=GradientPosition}
  n2 = MaterialFunctionCall!(Floor1, Floor2, Floor3, Basement) {materialFunction=/Game/Athena/WinterArt/Materials/MF_WS_GridUV}
  Add(A=@n1, B=@n2.BasementWorldHeight)
  FunctionInput() {inputName=1, inputType=FunctionInput_Vector4}
  FunctionInput() {inputName=2, inputType=FunctionInput_Vector4, sortPriority=1}
  FunctionInput() {inputName=3, inputType=FunctionInput_Vector4, sortPriority=2}
  FunctionInput() {inputName=4, inputType=FunctionInput_Vector4, sortPriority=3}
  n3 = FunctionInput() {inputName=MaterialIn, inputType=FunctionInput_MaterialAttributes}
  n4 = GetMaterialAttributes(None=@n3) {attributeGetTypes [#1]}
  n5 = VectorParameter() {group=Global, parameterName=WS GradientBottom, sortPriority=1, defaultValue=<rgba 0.4824 0.6547 0.7917 1>}
  n6 = VectorParameter() {group=Global, parameterName=WS GradientTop, sortPriority=1, defaultValue=<rgba 1 1 1 1>}
  n7 = Add(A=ScalarParameter() {defaultValue=2000, group=Global, parameterName=OffsetGradient}, B=@n1)
  n8 = Constant() {r=384}
  n9 = Ceil(None=ScalarParameter() {defaultValue=1, group=Global, parameterName=GradientFloors, sortPriority=3})
  n10 = Saturate(None=Divide(A=Subtract(A=@n2.BasementWorldHeight, B=@n7), B=Multiply(A=@n8, B=@n9)))
  n11 = Constant() {r=3}
  n12 = FunctionInput(Preview=ComponentMask(None=TextureCoordinate() {coordinateIndex=1}) {r=true}) {bUsePreviewValueAsDefault=true, inputName=transitionFactor, inputType=FunctionInput_Scalar, sortPriority=-1}
  n13 = If(
   A=Constant() {r=2}
   B=@n12
   "A > B"=If(
    A=Constant() {r=1}
    B=@n12
    "A > B"=ScalarParameter() {group=Global, parameterName=Show WS gradient on Texture 1?}
    "A < B"=ScalarParameter() {group=Global, parameterName=Show WS gradient on Texture 2?}
   ) {desc=0}
   "A < B"=ScalarParameter() {group=Global, parameterName=Show WS gradient on Texture 3?}
  )
  n14 = ScalarParameter() {group=Global, parameterName=Show WS gradient on Texture 4?}
  n15 = Multiply(
   A=@n4.BaseColor
   B=LinearInterpolate(A=Constant() {r=1}, B=LinearInterpolate(A=@n5.RGB, B=@n6.RGB, Alpha=@n10), Alpha=If(A=@n11, B=@n12, "A > B"=@n13, "A < B"=@n14))
  )
  n16 = ScalarParameter() {group=Global, parameterName=WS gradient mask A}
  n17 = ScalarParameter() {defaultValue=1, group=Global, parameterName=WS gradient mask B}
  n18 = TextureObjectParameter() {
   group=Global
   parameterName=WS_Gradient_Mask
   sortPriority=15
   texture=/Game/Environments/Sets/Pirate_Cove/Pirate_Ship/Textures/T_Hull_Breakup
  }
  n19 = ComponentMask(None=MaterialFunctionCall(TextureObject=@n18, TextureSize=ScalarParameter() {defaultValue=1493.3722, group=Global, parameterName=WS gradient mask scale}) {materialFunction=&WorldAlignedTexture}.XYZ Texture) {r=true}
  n20 = StaticSwitchParameter(True=LinearInterpolate(A=@n15, B=@n4.BaseColor, Alpha=LinearInterpolate(A=@n16, B=@n17, Alpha=@n19)), False=@n15) {parameterName=Use Gradient Mask?}
  FunctionOutput(None=StaticSwitchParameter(
    True=SetMaterialAttributes!(MaterialAttributes=@n4.MaterialAttributes, BaseColor=@n20) {attributeSetTypes [#1]}
    False=@n3
   ) {group=Global, parameterName=Use WS Gradient}) {outputName=Color*Gradient}
 ]}
